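# Idle Niagara effect for the Cat Burglar Winter Male pickaxe, adding icy glow, frost, snow and sparkles.
# Active emitters spawn at fixed offsets from User.SkeletalMesh samples, with cones aligned to sampled rotation spreading frost, sparkles and speckles.
# Glows layer three sprite renderers, movement-spawned frost uses SubUV sprites brightened by User.FrontendSpawn, and snow uses falling mesh particles.
# World-space snow and sparkle variants are present but disabled.
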NiagaraSystem /BRCosmetics/Weapons/FORT_Melee/Pickaxe_Cat_Burglar_Winter_Male/FX/NS_Cat_Burglar_Winter_Idle
dependencies {
 usedDataInterfaces [&NiagaraDataInterfaceArrayFloat3, &NiagaraDataInterfaceCurve]
 usedDynamicInputs []
 usedModules [&SystemState]
}
emitters {
 Empty {
  emitterUpdateScript [
   &EmitterState()
   &SpawnRate(Use Spawn Probability=false) {enabled=false}
   &SpawnBurst_Instantaneous(Spawn Count=2, Use Spawn Probability=true, UseLoopCountLimit=false)
  ]
  particleSpawnScript [
   &InitializeParticle(
    Color=<rgba 0.5 0.75 1 1>
    Lifetime Min=1.3
    Lifetime Mode=::NewEnumerator1
    Sprite Size Mode=::NewEnumerator1
    Uniform Sprite Size Max=40
    Uniform Sprite Size Min=32
    UsePositionOffset=false
    Write Material Random=true
   )
   &SkeletalMeshLocation(
    Interpolate Mesh Sampling=false
    MaskSpawnGroup=false
    Mesh Sampling Type=::NewEnumerator5
    OverrideRandomSeed=false
    Position Sampling=::NewEnumerator1
    Sampled Position Offset=&SelectVectorFromArray(
     Array Sampling Mode=::NewEnumerator0
     Direct Array Index=?
     Vector Selection Array=DataInterface() {propertyValues {FloatData [<0 5 -4>, <-11 5 -12.5>]}}
     Random Seed=0
    )
    Skeletal Mesh=$User.SkeletalMesh
   )
   &SphereLocation(
    MaskSpawnGroup=false
    Override Seed=false
    Sphere Radius=1.5
    Surface Only=false
    UseNonUniformScale=false
    UseOffset=false
    UseSphereOrientation=false
   ) {enabled=false}
   &CameraOffset(Camera Offset Amount=10)
  ]
  particleUpdateScript [
   &ParticleState()
   &ScaleColor(
    Scale Alpha=&FloatFromCurve(
     FloatCurve=DataInterface() {propertyValues {Curve {keys [
         {interpMode=RCIM_Cubic, tangentMode=RCTM_User, time=0, value=0}
         {arriveTangent=-2.4058, interpMode=RCIM_Cubic, leaveTangent=-2.4058, tangentMode=RCTM_User, time=0.26, value=0.8421}
         {arriveTangent=-1.1775, interpMode=RCIM_Cubic, leaveTangent=-1.1775, time=0.4613, value=0.0645}
         {arriveTangent=-0.1197, interpMode=RCIM_Cubic, leaveTangent=-0.1197, time=0.5898, value=0.4537}
         <at 1 0 RCIM_Cubic>
        ]}}}
     Scale Curve=&RandomRangeFloat(Minimum=0.1, Maximum=0.5)
    )
    Scale RGB=&VectorFromFloat(Value=&RandomRangeFloat(Minimum=0.6, Maximum=1, Evaluation Type=::NewEnumerator1))
    ScaleA=true
    ScaleRGB=true
   )
   &ScaleSpriteSize(Scale Factor=&MakeVector2D(
     X=&FloatFromCurve(
      FloatCurve=DataInterface() {propertyValues {Curve {keys [{interpMode=RCIM_Cubic, tangentMode=RCTM_User, time=0, value=0}, {interpMode=RCIM_Cubic, tangentMode=RCTM_User, time=1, value=1}]}}}
      Scale Curve=2
     )
     Y=&FloatFromCurve(
      FloatCurve=DataInterface() {propertyValues {Curve {keys [{interpMode=RCIM_Cubic, tangentMode=RCTM_User, time=0, value=1}, {interpMode=RCIM_Cubic, tangentMode=RCTM_User, time=1, value=0.6493}]}}}
      Scale Curve=2
     )
    ))
   /SetVariables(
    Particles.C:LinearColor=&MultiplyLinearColor(
     Linear Color=$Particles.Color
     Scale Factor=&MakeVector4_FromLinearColor(LinearColor=<rgba 5.84 10 9.2373 1>)
    )
    Particles.C1:LinearColor=&MultiplyLinearColor(
     Linear Color=$Particles.Color
     Scale Factor=&MakeVector4_FromLinearColor(LinearColor=<rgba 0.506 0.8106 1 1>)
    )
    Particles.NewLinearColor:LinearColor=&MultiplyLinearColor(
     Linear Color=$Particles.Color
     Scale Factor=&MakeVector4_FromLinearColor(LinearColor=<rgba 0.15 0.15 0.15 1>)
    )
    Particles.SpriteS1:Vector2f=&Multiply_Vector2(A=$Particles.SpriteSize, B=&Vector2DFromFloat(Value=0.3))
   ) {isSetParameters=true}
  ]
  rendererClasses [&NiagaraSpriteRendererProperties]
  renderers [
   {rendererClass=&NiagaraSpriteRendererProperties}
   {rendererClass=&NiagaraSpriteRendererProperties, rendererIndex=1}
   {rendererClass=&NiagaraSpriteRendererProperties, rendererIndex=2}
  ]
 }
 Frost {
  emitterUpdateScript [
   &EmitterState()
   &SpawnPerUnit(
    Max Movement Threshold:NiagaraFloat=&RandomRangeFloat(Minimum=5, Maximum=35)
    Module.Use Max Movement Threshold=true
    Module.Use Movement Tolerance=false
    Movement Tolerance=0.1
    Spawn Probability:NiagaraFloat=&RandomRangeFloat(Minimum=1, Maximum=2.5)
    Spawn Spacing:NiagaraFloat=&RandomRangeFloat(Minimum=0.35, Maximum=0.75)
    Use Spawn Probability=true
   )
   &SpawnRate(SpawnRate=500, Use Spawn Probability=false) {enabled=false}
  ]
  particleSpawnScript [
   /Niagara/Modules/Spawn/Initialization/InitializeParticle(
    Color=&MakeCustomLinearColorFromBool(Bool=$User.FrontendSpawn, True Color=<rgba 0.2779 0.6803 1 0.75>, False Color=<rgba 0.2779 0.6803 1 0.5>)
    Lifetime=&RandomRangeFloat(Minimum=1.3, Maximum=2)
    Module.Write Color=true
    Module.Write Lifetime=true
    Module.Write Mass=false
    Module.Write Material Random=true
    Module.Write Position=true
    Module.Write Scale=false
    Module.Write SpriteRotation=true
    Module.Write SpriteSize=true
    Sprite Rotation=-10
    Sprite Size=&RandomRangeVector2D(Minimum=<10 20>, Maximum=<20 35>)
   )
   &SkeletalMeshLocation(
    Interpolate Mesh Sampling=false
    MaskSpawnGroup=false
    Mesh Sampling Type=::NewEnumerator5
    OverrideRandomSeed=false
    Sampled Position Offset=<-50 5 -17>
    Skeletal Mesh=$User.SkeletalMesh
   ) {module=SkeletalMeshLocation001}
   &SpriteFacingAndAlignment(
    Sprite Alignment=&RotateVectorByQuaternion(Vector To Rotate=<1 0 0.2>, Quaternion=$Particles.SkeletalMeshLocation001.SampledRotation)
    Sprite Facing=<0 0 0>
    WriteAlignment=true
    WriteFacing=true
   )
   /Niagara/Modules/Spawn/Location/ConeLocation(
    Cone Angle=3
    Cone Axis:Vector3f=&RotateVectorByQuaternion(Vector To Rotate=<1 0 0.1>, Quaternion=$Particles.SkeletalMeshLocation001.SampledRotation)
    Cone Length=50
    Enable Wedge=true
    MaskSpawnGroup=false
    Module.Override Randomness=false
    Wedge Horizontal Angle=0
    Wedge Vertical Angle=10
   )
   &CameraOffset(Camera Offset Amount=20)
   &AddVelocity(Velocity=&RandomRangeVector(Minimum=<0 -5 0>, Maximum=<0 -10 0>))
  ]
  particleUpdateScript [
   &ParticleState()
   &SubUVAnimation(End Frame=40, Start Frame=2)
   &GravityForce(Gravity=&RandomRangeVector(Minimum=<0 0 -1>, Maximum=<0 0 -5>))
   &SolveForcesAndVelocity(Clamp Velocity=false, Limit Acceleration=false)
   &ScaleColor(
    Scale Alpha=&FloatFromCurve(FloatCurve=DataInterface() {propertyValues {Curve {keys [
         {arriveTangent=0.1402, interpMode=RCIM_Cubic, leaveTangent=0.1402, tangentMode=RCTM_User, time=0, value=0}
         {arriveTangent=0.317, interpMode=RCIM_Cubic, leaveTangent=0.317, tangentMode=RCTM_User, time=0.4, value=0.4}
         <at 0.7886 0 RCIM_Cubic>
        ]}}})
    ScaleA=true
    ScaleRGB=true
   )
   &DynamicMaterialParameters(
    Index 0 Param 4=1
    Module.Param0WriteEnabled=true
    Module.Param1WriteEnabled=true
    Module.Param2WriteEnabled=true
    Module.Param3WriteEnabled=true
   )
   &ScaleSpriteSize(Scale Factor=&Vector2DFromFloat(Value=&FloatFromCurve(FloatCurve=DataInterface() {propertyValues {Curve {keys [<at 0 1 RCIM_Cubic>, {arriveTangent=0.8108, interpMode=RCIM_Cubic, leaveTangent=0.8108, tangentMode=RCTM_User, time=1, value=1.5}]}}})))
  ]
  rendererClasses [&NiagaraSpriteRendererProperties]
  renderers [{rendererClass=&NiagaraSpriteRendererProperties}]
 }
 Snow_Local {
  emitterUpdateScript [
   &EmitterState()
   &SpawnRate(SpawnRate=6, Use Spawn Probability=false)
  ]
  particleSpawnScript [
   &InitializeParticle(
    Color=&MakeLinearColorFromVectorAndFloat(
     "Vector (RGB)"=&RandomRangeVector(Minimum=<0 1 1>, Maximum=<1 1 1>)
     "Float (Alpha)"=&RandomRangeFloat(Minimum=10, Maximum=13)
    )
    Lifetime=&RandomRangeFloat(Minimum=0.7, Maximum=1.3)
    Mass=0.5
    Mesh Scale=&Multiply_Vector(
     A=<0.15 0.025 0.1>
     B=&VectorFromFloat(Value=&RandomRangeFloat(Minimum=3, Maximum=7))
    )
    Module.Write Color=true
    Module.Write Lifetime=true
    Module.Write Mass=true
    Module.Write Material Random=true
    Module.Write Position=false
    Module.Write Scale=true
    Module.Write SpriteRotation=false
    Module.Write SpriteSize=true
    Sprite Rotation=&RandomRangeFloat(Minimum=0, Maximum=360)
    Sprite Size=<1 1>
   )
   &SkeletalMeshLocation(
    Interpolate Mesh Sampling=false
    MaskSpawnGroup=false
    Mesh Sampling Type=::NewEnumerator5
    Sampled Position Offset=<15 5 -10>
    Skeletal Mesh=$User.SkeletalMesh
    Socket Sampling Mode=::NewEnumerator5
   )
   &InitialMeshOrientation(
    Mesh Axis To Orient=<0 0 0>
    Mesh Orientation Mode=::NewEnumerator2
    Orientation Coordinate Space=::Local
    Orientation Vector=<0 0 0>
    Rotation=&RotateVectorByQuaternion(Vector To Rotate=<0.05 1 -1.1>, Quaternion=$Particles.SkeletalMeshLocation.SampledRotation)
    Use Rotation Vector=true
   )
   &DynamicMaterialParameters(
    Index 0 Param 2=&RandomRangeFloat(Minimum=30, Maximum=50)
    Module.Param0WriteEnabled=false
    Module.Param1WriteEnabled=true
    Module.Param2WriteEnabled=false
    Module.Param3WriteEnabled=true
   ) {module=DynamicMaterialParameters001}
   &AddVelocity(Velocity=&RandomRangeVector(Minimum=<0 -15 0>, Maximum=<0 -35 0>))
  ]
  particleUpdateScript [
   &GravityForce(Gravity=<0 0 -50>)
   &Drag(Drag=2, Use Linear Drag=true, Use Rotational Drag=false)
   &SolveForcesAndVelocity(Clamp Velocity=false, Limit Acceleration=false)
   &ParticleState()
   &DynamicMaterialParameters(
    Index 0 Param 1=&FloatFromCurve(FloatCurve=DataInterface() {propertyValues {Curve {keys [<at 0 10>,<at 1 0>]}}})
    Index 0 Param 2=&RandomRangeFloat(Minimum=30, Maximum=50, Evaluation Type=::NewEnumerator1)
    Index 0 Param 4=&FloatFromCurve(FloatCurve=DataInterface() {propertyValues {Curve {keys [<at 0 0.2>,<at 1 0>]}}})
    Module.Param0WriteEnabled=true
    Module.Param1WriteEnabled=false
    Module.Param2WriteEnabled=true
    Module.Param3WriteEnabled=true
   )
   &ScaleColor(
    Scale Alpha=&FloatFromCurve(
     FloatCurve=DataInterface() {propertyValues {Curve {keys [
         {arriveTangent=4.2631, interpMode=RCIM_Cubic, leaveTangent=4.2631, tangentMode=RCTM_User, time=0, value=0}
         <at 0.3504 1 RCIM_Cubic>
         {arriveTangent=-0.1318, interpMode=RCIM_Cubic, leaveTangent=-0.1318, tangentMode=RCTM_User, time=1, value=0}
        ]}}}
     Scale Curve=1.25
    )
    ScaleA=true
    ScaleRGB=true
   )
   &CameraOffset(Camera Offset Amount=10)
  ]
  rendererClasses [&NiagaraMeshRendererProperties]
  renderers [{rendererClass=&NiagaraMeshRendererProperties, rendererIndex=0}]
 }
 Snow_World {
  bEnabled=false
  emitterUpdateScript [
   &EmitterState()
   &SpawnRate(SpawnRate=6, Use Spawn Probability=false)
  ]
  particleSpawnScript [
   &InitializeParticle(
    Color=&MakeLinearColorFromVectorAndFloat(
     "Vector (RGB)"=&RandomRangeVector(Minimum=<0 1 1>, Maximum=<1 1 1>)
     "Float (Alpha)"=&RandomRangeFloat(Minimum=10, Maximum=13)
    )
    Lifetime=&RandomRangeFloat(Minimum=0.7, Maximum=1.3)
    Mass=0.5
    Mesh Scale=&Multiply_Vector(
     A=<0.15 0.025 0.1>
     B=&VectorFromFloat(Value=&RandomRangeFloat(Minimum=3, Maximum=7))
    )
    Module.Write Color=true
    Module.Write Lifetime=true
    Module.Write Mass=true
    Module.Write Material Random=true
    Module.Write Position=false
    Module.Write Scale=true
    Module.Write SpriteRotation=false
    Module.Write SpriteSize=true
    Sprite Rotation=&RandomRangeFloat(Minimum=0, Maximum=360)
    Sprite Size=<1 1>
   )
   &AddVelocity(Scale Added Velocity=<0 0 1>, Velocity=&RandomRangeVector(Minimum=<50 0 0>, Maximum=<100 0 0>))
   &SkeletalMeshLocation(
    Interpolate Mesh Sampling=false
    MaskSpawnGroup=false
    Mesh Sampling Type=::NewEnumerator5
    Sampled Orientation Coordinate Space=::World
    Sampled Position Offset=<15 5 -10>
    Skeletal Mesh=$User.SkeletalMesh
    Socket Sampling Mode=::NewEnumerator5
   )
   &InitialMeshOrientation(
    Mesh Axis To Orient=<0 0 0>
    Mesh Orientation Mode=::NewEnumerator2
    Orientation Vector=<0 0 0>
    Rotation=&RotateVectorByQuaternion(Vector To Rotate=<0.05 1 -1.1>, Quaternion=$Particles.SkeletalMeshLocation.SampledRotation)
    Use Rotation Vector=true
   )
   &DynamicMaterialParameters(
    Index 0 Param 2=&RandomRangeFloat(Minimum=30, Maximum=50)
    Module.Param0WriteEnabled=false
    Module.Param1WriteEnabled=true
    Module.Param2WriteEnabled=false
    Module.Param3WriteEnabled=true
   ) {module=DynamicMaterialParameters001}
  ]
  particleUpdateScript [
   &GravityForce(Gravity=<0 0 -50>)
   &Drag(Drag=5, Use Linear Drag=true, Use Rotational Drag=false)
   &InheritSourceMovement(Applied Position Scale=<0.25 0.25 0.25>, ApplyForce=false, ApplyPosition=true, ApplyPositionFalloff=false)
   &SolveForcesAndVelocity(Clamp Velocity=false, Limit Acceleration=false)
   &ParticleState()
   &DynamicMaterialParameters(
    Index 0 Param 1=&FloatFromCurve(FloatCurve=DataInterface() {propertyValues {Curve {keys [<at 0 10>,<at 1 0>]}}})
    Index 0 Param 2=&RandomRangeFloat(Minimum=30, Maximum=50, Evaluation Type=::NewEnumerator1)
    Index 0 Param 4=&FloatFromCurve(FloatCurve=DataInterface() {propertyValues {Curve {keys [<at 0 0.2>,<at 1 0>]}}})
    Module.Param0WriteEnabled=true
    Module.Param1WriteEnabled=false
    Module.Param2WriteEnabled=true
    Module.Param3WriteEnabled=true
   )
   &ScaleColor(
    Scale Alpha=&FloatFromCurve(FloatCurve=DataInterface() {propertyValues {Curve {keys [
         {arriveTangent=4.2631, interpMode=RCIM_Cubic, leaveTangent=4.2631, tangentMode=RCTM_User, time=0, value=0}
         <at 0.3504 1 RCIM_Cubic>
         {arriveTangent=-0.1318, interpMode=RCIM_Cubic, leaveTangent=-0.1318, tangentMode=RCTM_User, time=1, value=0}
        ]}}})
    ScaleA=true
    ScaleRGB=true
   )
   &CameraOffset(Camera Offset Amount=10)
  ]
  rendererClasses [&NiagaraMeshRendererProperties]
  renderers [{rendererClass=&NiagaraMeshRendererProperties, rendererIndex=0}]
 }
 Sparkle_Ice_Local {
  emitterUpdateScript [
   &EmitterState(
    Life Cycle Mode=::NewEnumerator1
    Loop Duration=&RandomRangeFloat(Minimum=0.5, Maximum=2)
    UseLoopDelay=false
   )
   &SpawnBurst_Instantaneous(
    Spawn Count=&RandomRangeInt(Minimum=0, Maximum=1)
    Use Spawn Probability=false
    UseLoopCountLimit=false
   )
  ]
  particleSpawnScript [
   /Niagara/Modules/Spawn/Initialization/V2/InitializeParticle(
    Color=<rgba 1 1.2 2 1>
    Lifetime Mode=::NewEnumerator1
    Sprite Size Mode=::NewEnumerator1
    Uniform Sprite Size Max=20
    Uniform Sprite Size Min=15
    UsePositionOffset=false
    Write Material Random=true
   )
   &SkeletalMeshLocation(
    Interpolate Mesh Sampling=false
    MaskSpawnGroup=false
    Mesh Sampling Type=::NewEnumerator5
    OverrideRandomSeed=false
    Sampled Position Offset=<-70 5 -17>
    Skeletal Mesh=$User.SkeletalMesh
   )
   &CameraOffset(Camera Offset Amount=20)
   &ConeLocation(
    Cone Angle=3
    Cone Axis:Vector3f=&RotateVectorByQuaternion(Vector To Rotate=<1 0 0.1>, Quaternion=$Particles.SkeletalMeshLocation.SampledRotation)
    Cone Length=60
    Enable Wedge=true
    MaskSpawnGroup=false
    Module.Override Randomness=false
    Wedge Horizontal Angle=0
    Wedge Vertical Angle=10
   )
   &DynamicMaterialParameters(
    Index 0 Param 1=0.4
    Index 0 Param 2=1
    Index 0 Param 3=1
    Index 0 Param 4=1
    Module.Param0WriteEnabled=true
    Module.Param1WriteEnabled=true
    Module.Param2WriteEnabled=true
    Module.Param3WriteEnabled=true
   )
  ]
  particleUpdateScript [
   &ParticleState()
   &ScaleColor(Scale Alpha=&RandomRangeFloat(Minimum=0.25, Maximum=0.75), ScaleA=true, ScaleRGB=true)
   &ScaleSpriteSize(Uniform Curve Sprite Scale=DataInterface() {propertyValues {Curve {keys [{tangentMode=RCTM_User, time=0, value=0}, <at 0.18 1>, <at 0.6 1>, {tangentMode=RCTM_User, time=1, value=0}]}}})
   &SpriteRotationRate(Rotation Rate=&RandomRangeFloat(Minimum=-25, Maximum=25))
  ]
  rendererClasses [&NiagaraSpriteRendererProperties]
  renderers [{rendererClass=&NiagaraSpriteRendererProperties}]
 }
 Sparkle_Ice_World {
  bEnabled=false
  emitterUpdateScript [
   &EmitterState(Life Cycle Mode=::NewEnumerator1, Loop Duration=0.33, UseLoopDelay=false)
   &SpawnRate(SpawnRate=3, Use Spawn Probability=false) {enabled=false}
   &SpawnBurst_Instantaneous(
    Spawn Count=&RandomRangeInt(Minimum=0, Maximum=6)
    Use Spawn Probability=false
    UseLoopCountLimit=false
   )
  ]
  particleSpawnScript [
   &InitializeParticle(
    Color=<rgba 0.1 0.75 5 1>
    Lifetime Max=1
    Lifetime Min=0.5
    Lifetime Mode=::NewEnumerator1
    Sprite Size Mode=::NewEnumerator1
    Uniform Sprite Size Max=25
    Uniform Sprite Size Min=16
    UsePositionOffset=false
    Write Material Random=true
   )
   &SkeletalMeshLocation(
    Interpolate Mesh Sampling=false
    MaskSpawnGroup=false
    Mesh Sampling Type=::NewEnumerator5
    OverrideRandomSeed=false
    Sampled Orientation Coordinate Space=::World
    Sampled Position Offset=<-70 5 -17.5>
    Skeletal Mesh=$User.SkeletalMesh
   )
   &CameraOffset(Camera Offset Amount=10)
   &ConeLocation(
    Cone Angle=3
    Cone Axis:Vector3f=&RotateVectorByQuaternion(Vector To Rotate=<1 0 0.1>, Quaternion=$Particles.SkeletalMeshLocation.SampledRotation)
    Cone Length=70
    Coordinate Space:ENiagaraCoordinateSpace=::World
    Enable Wedge=true
    MaskSpawnGroup=false
    Module.Override Randomness=false
    Wedge Horizontal Angle=10
    Wedge Vertical Angle=0
   )
   &DynamicMaterialParameters(
    Index 0 Param 1=0.4
    Index 0 Param 2=1
    Index 0 Param 3=1
    Index 0 Param 4=1
    Module.Param0WriteEnabled=true
    Module.Param1WriteEnabled=true
    Module.Param2WriteEnabled=true
    Module.Param3WriteEnabled=true
   )
  ]
  particleUpdateScript [
   &ParticleState()
   &ScaleColor(Scale Alpha=&RandomRangeFloat(Minimum=0.5, Maximum=1.25), ScaleA=true, ScaleRGB=true)
   &ScaleSpriteSize(Uniform Curve Sprite Scale=DataInterface() {propertyValues {Curve {keys [{tangentMode=RCTM_User, time=0, value=0}, <at 0.18 1>, <at 0.6 1>, {tangentMode=RCTM_User, time=1, value=0}]}}})
   &SpriteRotationRate(Rotation Rate=&RandomRangeFloat(Minimum=-25, Maximum=25))
  ]
  rendererClasses [&NiagaraSpriteRendererProperties]
  renderers [{rendererClass=&NiagaraSpriteRendererProperties}]
 }
 speckles_ice {
  emitterUpdateScript [
   &EmitterState()
   &SpawnRate(SpawnRate=&RandomRangeFloat(Minimum=-4, Maximum=12), Use Spawn Probability=false)
  ]
  particleSpawnScript [
   &InitializeParticle(
    Color=<rgba 1 1.5 2 0.75>
    Lifetime Max=0.8
    Lifetime Min=0.35
    Lifetime Mode=::NewEnumerator1
    Sprite Size Mode=::NewEnumerator1
    Uniform Sprite Size Max=6
    Uniform Sprite Size Min=2
    UsePositionOffset=false
    Write Material Random=true
   )
   &SkeletalMeshLocation(
    Interpolate Mesh Sampling=false
    MaskSpawnGroup=false
    Mesh Sampling Type=::NewEnumerator5
    OverrideRandomSeed=false
    Sampled Position Offset=<-70 5 -17>
    Skeletal Mesh=$User.SkeletalMesh
   ) {module=SkeletalMeshLocation001}
   &CameraOffset(Camera Offset Amount=15)
   &ConeLocation(
    Cone Angle=3
    Cone Axis:Vector3f=&RotateVectorByQuaternion(Vector To Rotate=<1 0 0.1>, Quaternion=$Particles.SkeletalMeshLocation001.SampledRotation)
    Cone Length=60
    Enable Wedge=true
    MaskSpawnGroup=false
    Module.Override Randomness=false
    Wedge Horizontal Angle=0
    Wedge Vertical Angle=10
   )
   &DynamicMaterialParameters(
    Index 0 Param 1=0.4
    Index 0 Param 2=1
    Index 0 Param 3=1
    Index 0 Param 4=1
    Module.Param0WriteEnabled=true
    Module.Param1WriteEnabled=true
    Module.Param2WriteEnabled=true
    Module.Param3WriteEnabled=true
   )
  ]
  particleUpdateScript [
   &ParticleState()
   &ScaleColor(
    Scale Alpha=&FloatFromCurve(FloatCurve=DataInterface() {propertyValues {Curve {keys [
         {interpMode=RCIM_Cubic, tangentMode=RCTM_User, time=0, value=0}
         {interpMode=RCIM_Cubic, tangentMode=RCTM_User, time=0.4, value=1}
         {tangentMode=RCTM_User, time=1, value=0}
        ]}}})
    ScaleA=true
    ScaleRGB=true
   )
   &ScaleSpriteSize(Uniform Curve Sprite Scale=DataInterface() {propertyValues {Curve {keys [{tangentMode=RCTM_User, time=0, value=0}, <at 0.18 1>, <at 0.6 1>, {tangentMode=RCTM_User, time=1, value=0}]}}})
   &SpriteRotationRate(Rotation Rate=&RandomRangeFloat(Minimum=-25, Maximum=25))
  ]
  rendererClasses [&NiagaraSpriteRendererProperties]
  renderers [{rendererClass=&NiagaraSpriteRendererProperties}]
 }
}
summary {
 systemName=NS_Cat_Burglar_Winter_Idle
 emitters [
  {emitterName=Empty, rendererClasses [&NiagaraSpriteRendererProperties]}
  {emitterName=Sparkle_Ice_Local, rendererClasses [&NiagaraSpriteRendererProperties]}
  {emitterName=speckles_ice, rendererClasses [&NiagaraSpriteRendererProperties]}
  {emitterName=Snow_Local, rendererClasses [&NiagaraMeshRendererProperties]}
  {bEnabled=false, emitterName=Sparkle_Ice_World, rendererClasses [&NiagaraSpriteRendererProperties]}
  {emitterName=Frost, rendererClasses [&NiagaraSpriteRendererProperties]}
  {bEnabled=false, emitterName=Snow_World, rendererClasses [&NiagaraMeshRendererProperties]}
 ]
 userVariables [
  User.FrontendSpawn {type=NiagaraBool}
  User.RibbonIDOffset {type=NiagaraInt32}
  User.SkeletalMesh {
   type=NiagaraDataInterfaceSkeletalMesh
   defaultValue {
    struct=NiagaraExt_VariableValue_DataInterface
    value {dataInterface=&self:NiagaraDataInterfaceSkeletalMesh_0}
   }
  }
 ]
}
